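# UI material for an item shop tile that adds tinted, drifting smoke over a background texture.
# A flowmap with panning, fill-scaled UVs distorts blurred noise, which is blended with two noise layers panning at different speeds.
# The tinted result is faded by a linear gradient with adjustable endpoints and angle, then added to the background.
# A static switch selects the background's UV mapping.
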Material /ItemShop/UI/Art/Refeatures/M_UI_Smoke_ShopTile
graph {
 properties {bUsedWithStaticMesh=false, materialDomain=MD_UI, pixelDepthOffsetMode=PDOM_Legacy, usedShadingModels=Default Lit}
 nodes [
  n1 = NamedRerouteUsage()
  n2 = NamedRerouteUsage()
  n3 = TextureObjectParameter() {
   group=Fog Flowmap
   parameterName=Fog Shape - Flowmap
   samplerType=SAMPLERTYPE_LinearColor
   texture=/Game/Global/Textures/Default/Flowmaps/T_SwirlFlow_02
  }
  n4 = ScalarParameter() {defaultValue=1, group=Fog Flowmap, parameterName=Flowmap Size}
  n5 = MaterialFunctionCall() {materialFunction=&MF_UI_UVScaling}
  n6 = ScalarParameter() {group=Fog Flowmap, parameterName=Flowmap Movement X}
  n7 = ScalarParameter() {defaultValue=0.02, group=Fog Flowmap, parameterName=Flowmap Movement Y}
  n8 = MaterialFunctionCall!(
   Texture 2D=TextureObject() {samplerType=SAMPLERTYPE_Grayscale, texture=&LowResBlurredNoise}
   Texture UVs
   FlowMap Texture 2D=@n3
   FlowMap UVs=Panner(Coordinate=Multiply(A=@n5.Scaled To Fill, B=@n4) {constB=0.75}, Speed=AppendVector(A=@n6, B=@n7)) {speedY=0.03}
   FlowIntensity=ScalarParameter() {defaultValue=0.2, group=Fog Flowmap, parameterName=Flowmap Intensity}
   FlowSpeed
  ) {materialFunction=/FortUILibrary/MaterialFunctions/Effects/MF_UI_FlowMap}
  n9 = Multiply(A=@n5.Scaled To Fill) {constB=2}
  n10 = Multiply(A=Time(), B=ScalarParameter() {defaultValue=3, group=Fog, parameterName=Cloud Movement})
  n11 = Panner(Coordinate=@n9, Time=@n10) {speedX=0.015, speedY=0.013}
  n12 = Panner(Coordinate=@n9, Time=@n10) {speedX=0.05, speedY=0.015}
  n13 = Multiply(A=Add(
    A=TextureSample(UVs=@n11) {samplerSource=SSM_Wrap_WorldGroupSettings, texture=/Game/UI/Frontend/ItemShop/Art/IWDTakeover/Texture/T-Noise-IWD}.R
    B=TextureSample(UVs=@n12) {samplerType=SAMPLERTYPE_Grayscale, texture=&LowResBlurredNoise}.R
   )) {constB=1.4}
  n14 = Multiply(
   A=VectorParameter() {group=Fog, parameterName=Color, defaultValue=<rgba 0.151 0.0484 0.1094 1>}.RGB
   B=LinearInterpolate(B=LinearInterpolate(A=@n8.Result, Alpha=@n13), Alpha=@n8.Result)
  )
  n15 = ScalarParameter() {defaultValue=2, group=Fog, parameterName=Smoke Opacity}
  n16 = ScalarParameter() {defaultValue=0.25, group=Fog, parameterName=Smoke Endpoint}
  n17 = ScalarParameter() {defaultValue=1.52, group=Fog, parameterName=Smoke Intensity}
  n18 = Constant()
  n19 = MaterialFunctionCall(
   Input=ScalarParameter() {defaultValue=90, group=Fog, parameterName=Smoke Angle}
   Input Low=@n18
   Input High=Constant() {r=-360}
   Target Low=@n18
   Target High=Constant() {r=1}
  ) {materialFunction=&RemapValueRange}
  Add(
   A=TextureSampleParameter2D(UVs=StaticSwitchParameter(True=@n1, False=@n2) {group=Texture, parameterName=UVScaling}) {parameterName=Background Texture, texture=/OfferCatalog/Art/FNM/S26/T_UI_MarbleFloor}.RGB
   B=Multiply(
    A=Multiply(A=@n14, B=@n15) {constB=0.3}
    B=MaterialFunctionCall(GradientStart=@n16, GradientEnd=@n17, RotationAngle=@n19.Result) {materialFunction=&MF_UI_LinearGradient}.Result
   )
  )
  NamedRerouteDeclaration(None=@n5.Scaled to Fit) {name=UVFit, nodeColor=<rgba 1 0 0.7266 1>}
  NamedRerouteDeclaration(None=MaterialFunctionCall() {materialFunction=&ScreenAlignedUVs}."X 100%, Y 100%") {name=ScreenAlignedUVs, nodeColor=<rgba 0 0.6484 1 1>}
 ]
}
